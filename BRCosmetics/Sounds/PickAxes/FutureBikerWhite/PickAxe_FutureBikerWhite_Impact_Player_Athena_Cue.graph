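# Sound cue for the FutureBikerWhite cosmetic pickaxe's player impact sound in Battle Royale.
# A Random node at the root picks among six equally weighted impact wave variations.
# The cue primes on load, plays at reduced volume in the impacts sound class, and references shared pickaxe impact attenuation, a cosmetic pickaxe audio property sheet and the player melee weapon concurrency set.
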SoundCue /BRCosmetics/Sounds/PickAxes/FutureBikerWhite/PickAxe_FutureBikerWhite_Impact_Player_Athena_Cue
root=Random_0
nodes [
 {id=Random_0, children [WavePlayer_0,WavePlayer_1,WavePlayer_2,WavePlayer_3,WavePlayer_4,WavePlayer_9], properties {weights [1,1,1,1,1,1]}}
 {id=WavePlayer_0, properties {soundWaveAssetPtr=&/FutureBikerWhite_Impact_01}}
 {id=WavePlayer_1, properties {soundWaveAssetPtr=&/FutureBikerWhite_Impact_02}}
 {id=WavePlayer_2, properties {soundWaveAssetPtr=&/FutureBikerWhite_Impact_03}}
 {id=WavePlayer_3, properties {soundWaveAssetPtr=&/FutureBikerWhite_Impact_04}}
 {id=WavePlayer_4, properties {soundWaveAssetPtr=&/FutureBikerWhite_Impact_05}}
 {id=WavePlayer_9, properties {soundWaveAssetPtr=&/FutureBikerWhite_Impact_06}}
]
properties {
 attenuationSettings=/Game/Sounds/Attenuation/Fort_Pickaxe_Impact_Att
 audioPropertiesSheet=/BRCosmetics/Sounds/Settings/PropertySheets/Pickaxe/APS_Cosmetic_Pickaxe_Impact
 bPrimeOnLoad=true
 maxDistance=4300
 soundClassObject=&Fort_Impacts
 volumeMultiplier=0.6
 attenuationOverrides {
  bAttenuateWithLPF=true
  distanceAlgorithm=Logarithmic
  falloffDistance=800
  lPFRadiusMax=800
  lPFRadiusMin=400
  attenuationShapeExtents {x=300}
 }
 concurrencyOverrides {bLimitToOwner=true, maxCount=2, resolutionRule=StopOldest}
 concurrencySet [&Player_Melee_Weapon_Conc]
}
